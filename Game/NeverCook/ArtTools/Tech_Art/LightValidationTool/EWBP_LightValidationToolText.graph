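# Editor utility widget that serves as a small text label, apparently for the LightValidationTool in the tech art tools.
# An Overlay holds a single grey, bold Roboto TextBlock.
# On PreConstruct it copies its Text string variable, empty by default, into that TextBlock.
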EditorUtilityWidgetBlueprint /Game/NeverCook/ArtTools/Tech_Art/LightValidationTool/EWBP_LightValidationToolText
graphs [EventGraph {nodes [
   Event PreConstruct()
   > "SetText (Text)"(self=Get TextBlock_55() -> (TextBlock_55:Text Object Reference), InText=Get Text!() -> (Text))
  ]}]
variables [Text {type=string, default=""}]
widgets {
 info {rootWidgetClass=&Overlay, widgetCount=2}
 widgets [
  Overlay_0 {widgetClassPath=&Overlay, widget=Overlay_0}
  TextBlock_55 {
   widgetClassPath=&TextBlock
   bIsVariable=true
   parent=Overlay_0
   widget=TextBlock_55
   properties {
    text=Text Block
    colorAndOpacity {colorUseRule=UseColor_Specified, specifiedColor=<rgba 0.5938 0.5938 0.5938 1>}
    font {
     bForceMonospaced=false
     bMaterialIsStencil=false
     fontMaterial=null
     fontObject=&Roboto
     letterSpacing=0
     monospacedWidth=1
     size=9.75
     skewAmount=0
     typefaceFontName=Bold
     outlineSettings {
      bApplyOutlineToDropShadows=false
      bMiteredCorners=false
      bSeparateFillAlpha=false
      outlineMaterial=null
      outlineSize=0
      outlineColor=<rgba 0 0 0 1>
     }
    }
   }
   slot {class=OverlaySlot}
  }
 ]
}
