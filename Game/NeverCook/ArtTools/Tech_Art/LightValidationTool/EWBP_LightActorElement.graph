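# Editor utility row widget for the light validation tool, representing one actor's lights in the level.
# On construct it lists components with dirty-property tooltips (or warns on plain Light actors); a Python query supplies the last user to change it.
# Find selects the actor and frames the viewport using a trace and bounds focus.
# Collapsed Restore resets light properties, looping until Python finds none dirty, then fires OnResetProperty.
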EditorUtilityWidgetBlueprint /Game/NeverCook/ArtTools/Tech_Art/LightValidationTool/EWBP_LightActorElement
graphs [
 EventGraph {nodes [
   "On Clicked (EditorUtilityButton)"() {is=title K2Node_ComponentBoundEvent}
   > n1 = Set DynaComp(DynaComp:Array of Light Component Object References=Get Components() -> (Components:Array of Light Component Object References)) -> (Output_Get:Array of Light Component Object References)
   n2 = For Each Loop(
    Exec=[@n5.then, @n1.then]
    Array:Array of Light Component Object References=Get DynaComp() -> (DynaComp:Array of Light Component Object References)
   ) -> (Array Element:Light Component Object Reference)
   n3 = Execute Python Script!(
    execute=@n2.Completed
    PythonScript="import sys\r\npath = project_content_dir+\"/Python/TechArt/Tools\"\r\nsys.path.append(project_content_dir+\"&Tools/\")\r\nimport LightValidatorUtils\r\nimport importlib\r\n##importlib.reload(LightValidatorUtils)\r\ncomponents_filtered = LightValidatorUtils.CheckIfStillHasDirty(actor)"
    project_content_dir:String=ProjectContentDir!() -> (ReturnValue:String)
    actor:Actor Object Reference=Get Actor() -> (Actor:Actor Object Reference)
   ) -> (then, ReturnValue, filtered_components:Array of Light Component Object References) {inputs [project_content_dir, actor], outputs [filtered_components]}
   > n4 = Set DynaComp(DynaComp=@n3.filtered_components)
   > n5 = Branch(Condition="integer > integer"(A=Length(TargetArray:Array of Light Component Object References=@n4)))
   Call On Reset Property!(execute:Exec=@n5.else, self:EWBP Light Actor Element Object Reference) -> (then:Exec) {is=title K2Node_CallDelegate}
   n6 = Execute Python Script!(
    execute=@n2.LoopBody
    PythonScript="import sys\r\npath = project_content_dir+\"&Tools\"\r\nsys.path.append(project_content_dir+\"&Tools/\")\r\nimport LightValidatorUtils\r\nimport importlib\r\n##importlib.reload(LightValidatorUtils)\r\nproperties_name = LightValidatorUtils.light_properties"
    project_content_dir=ProjectContentDir()
   ) -> (then, ReturnValue, propertiesName:Array of Strings) {inputs [project_content_dir], outputs [propertiesName]}
   > n7 = For Each Loop(Array:Array of Strings=@n6.propertiesName) -> (Array Element:String)
   > ResetEditorProperty!(execute, Object:Object Reference=@n2, PropertyName:Name=String To Name(InString=@n7), ChangeNotifyMode:EPropertyAccessChangeNotifyMode Enum) -> (then, ReturnValue:bool)
   "On Clicked (EditorUtilityButton_270)"() {is=title K2Node_ComponentBoundEvent}
   n8 = Get Components()
   > DrawDebugSphere!(
    execute
    WorldContextObject:Object Reference
    Center:Vector=Get World Location!(self:Scene Component Object Reference="Get (a copy)"(Array:Array of Light Component Object References=@n8) -> (Output:Light Component Object Reference)) -> (ReturnValue:Vector) {is=title K2Node_CallFunction}
    Radius:float=50
    Segments:int=12
    LineColor:Linear Color Structure=(R=1,G=0.1332,B=0.1,A=1)
    Duration:float=10
    Thickness:float=1
    DepthPriority:EDrawDebugSceneDepthPriorityGroup Enum=World
   ) -> (then)
   > SetSelectedLevelActors(
    self=Get EditorActorSubsystem()
    ActorsToSelect=Make Array("[0]":Actor Object Reference=Get Actor()) -> (Array:Array of Actor Object References)
   )
   n9 = Get Actor()
   n10 = Get Actor Location(self=@n9)
   n11 = Get Actor Rotation!(self:Actor Object Reference=@n9) -> (ReturnValue:Rotator) {is=title K2Node_CallFunction}
   > Set EndRay(EndRay:Vector=Forward Target Offset!(Origin:Vector=@n10, Rotation:Rotator=@n11, Offset:Vector="-500,0,0") -> (Return Transform:Transform, Return Start Location:Vector=~, Return End Location:Vector=~) {is=title K2Node_MacroInstance}.Return End Location) -> (Output_Get:Vector)
   > n12 = Sequence()
   > Get Editor Camera Info!(execute) -> (then, CameraLocation:Vector, ForwardDirection:Vector, FOV:float, ReturnValue:bool) {is=title K2Node_CallFunction}
   > n13 = Line Trace For Objects!(
    execute
    WorldContextObject:Object Reference
    Start:Vector=Forward Target Offset(Origin=@n10, Rotation=@n11, Offset="0.5,0.5,0.5") -> (Return Start Location=~, Return End Location=~).Return End Location
    End:Vector=Get EndRay() -> (EndRay:Vector)
    ObjectTypes:Array of EObjectTypeQuery Enums=Make Array("[0]":EObjectTypeQuery Enum=ObjectTypeQuery1) -> (Array:Array of EObjectTypeQuery Enums)
    bTraceComplex:bool=true
    ActorsToIgnore:Array of Actor Object References
    DrawDebugType:EDrawDebugTrace Enum=Persistent
    bIgnoreSelf:bool=true
    TraceColor:Linear Color Structure=(R=1,G=0,B=0,A=1)
    TraceHitColor:Linear Color Structure=(R=0,G=1,B=0,A=1)
    DrawTime:float=5
   ) -> (then, OutHit:Hit Result Structure, ReturnValue:bool)
   n14 = Get Actor()
   n15 = BreakHitResult!(Hit:"Hit Result Structure (by ref)"=@n13) -> (
    bBlockingHit:bool
    bInitialOverlap:bool
    Time:float
    Distance:float
    Location:Vector
    ImpactPoint:Vector
    Normal:Vector
    ImpactNormal:Vector
    PhysMat:Physical Material Object Reference
    HitActor:Actor Object Reference
    HitComponent:Primitive Component Object Reference
    HitBoneName:Name
    BoneName:Name
    HitItem:int
    ElementIndex:int
    FaceIndex:int
    TraceStart:Vector
    TraceEnd:Vector
   )
   n16 = vector - vector(A=@n15.TraceStart, B=@n15.TraceEnd)
   n17 = Look At Function!(
    CurrentTransform_Location:Vector=vector + vector(A=vector / vector(A:Vector=@n16, B=2) -> (ReturnValue:Vector), B=@n15.TraceStart)
    CurrentTransform_Rotation:Rotator=Get Actor Rotation(self=@n14)
    CurrentTransform_Scale:Vector="1,1,1"
    TargetPosition:"Vector (by ref)"=Get Actor Location(self=@n14)
    LookAtVector:Vector=vector * vector(A=GetActorForwardVector!(self:Actor Object Reference=@n14) -> (ReturnValue:Vector), B=-1) -> (ReturnValue=~)
    bUseUpVector:bool=false
    UpVector:Vector=GetActorUpVector!(self:Actor Object Reference=@n14) -> (ReturnValue:Vector)
    ClampConeInDegree:float=0
   ) -> (ReturnValue_Location:Vector, ReturnValue_Rotation:Rotator, ReturnValue_Scale:Vector)
   > SetLevelViewportCameraInfo!(
    execute
    self:Unreal Editor Subsystem Object Reference=Get UnrealEditorSubsystem()
    CameraLocation:Vector=@n17
    CameraRotation:Rotator=@n17.ReturnValue_Rotation
   ) -> (then)
   Set Timer by Event(execute=@n12.then_1, Delegate=Create Event(), Time=0.01, bLooping=false, bMaxOncePerFrame=true)
   Align with delay() {is=title K2Node_CustomEvent}
   n18 = Get Actor Bounds(self=Get Actor(), bOnlyCollidingComponents=false)
   > Focus Bounds in Viewport!(" ":Exec, center:Vector=@n18, bounds:Vector=vector / vector(A=@n18.BoxExtent, B=2)) -> ("  ":Exec) {is=title K2Node_MacroInstance}
   QueryLastChangeBy() {is=title K2Node_CustomEvent}
   > n19 = Execute Python Script!(
    execute
    PythonScript="import sys\r\npath = project_content_dir+\"&Tools\"\r\nsys.path.append(project_content_dir+\"&Tools/\")\r\nimport LightValidatorUtils\r\nimport importlib\r\nimportlib.reload(LightValidatorUtils)\r\nchanged_by =LightValidatorUtils.GetLastUser(actor)"
    project_content_dir=ProjectContentDir()
    actor:Actor Object Reference=Get Actor()
   ) -> (then, ReturnValue, changedBy:String) {inputs [project_content_dir, actor], outputs [changedBy]}
   n20 = Split(SourceString=@n19.changedBy, InStr=++, SearchCase=CaseSensitive, SearchDir=FromStart)
   n21 = Get UserNameText() -> (UserNameText:Text Object Reference)
   > "SetText (Text)"(self=@n21, InText="To Text (String)"(InString=@n20))
   > Set ToolTipText!(execute, ToolTipText:Text="To Text (String)"(InString=@n20.RightS), self:Widget Object Reference=@n21) -> (then, Output_Get:Text)
   Event Construct()
   n22 = Get Actor()
   n23 = Get Actor()
   n24 = Split(SourceString=Split(SourceString=Get Object Path String(Object=@n23), InStr=/, SearchDir=FromEnd).RightS, InStr=., SearchDir=FromStart)
   > "SetText (Text)"(
    self=Get ActorNameText() -> (ActorNameText:Text Object Reference)
    InText=Format Text!(
     Format="INVTEXT(\"Level -> {LevelName}\\r\\nActor -> {name}\")"
     LevelName:String=Split(SourceString=@n24.RightS, InStr=:, SearchCase=CaseSensitive, SearchDir=FromStart)
     name:String=GetDisplayName(Object=@n22)
    ) -> (Result)
   )
   > ClearChildren(self=Get ComponentsVerticalBox() -> (ComponentsVerticalBox:Vertical Box Object Reference))
   > n25 = Cast To Light!(execute, Object=Get Actor()) -> (then, CastFailed, AsLight:Light Object Reference) {is=title K2Node_DynamicCast}
   > n26 = Create Widget!(
    execute:Exec
    WorldContextObject:Object Reference
    Class:User Widget Class Reference=&/EWBP_LightValidationToolText.EWBP_LightValidationToolText_C
    OwningPlayer:Player Controller Object Reference
    Text:Text="INVTEXT(\"WARNING! THIS IS LIGHT ACTOR NOT CUSTOM BP!\")"
   ) -> (then:Exec, ReturnValue:EWBP Light Validation Tool Text Object Reference)
   > Add Unique(
    TargetArray:Array of EWBP Light Validation Tool Text Object References=Get TextComponents() -> (TextComponents:Array of EWBP Light Validation Tool Text Object References)
    NewItem:"EWBP Light Validation Tool Text Object Reference (by ref)"=@n26
   )
   > n27 = AddChild(self=Get ComponentsVerticalBox(), Content=@n26)
   Set Timer by Event(execute=[@n29.Completed, @n27.then], Delegate=Create Event(), Time=Get Delay() -> (Delay:double), bLooping=false, bMaxOncePerFrame=true)
   n28 = For Each Loop(Exec=@n25.CastFailed, Array:Array of Light Component Object References=Get Components()) -> (Array Element:Light Component Object Reference)
   n29 = For Each Loop(Exec=@n28.Completed, Array:Array of EWBP Light Validation Tool Text Object References=Get TextComponents()) -> (Array Element:EWBP Light Validation Tool Text Object Reference)
   > n30 = Execute Python Script!(
    execute
    PythonScript="import sys\r\npath = project_content_dir+\"&Tools\"\r\nsys.path.append(project_content_dir+\"&Tools/\")\r\nimport LightValidatorUtils\r\n##import importlib\r\n##importlib.reload(LightValidatorUtils)\r\nfiltered_properties = LightValidatorUtils.GrabDirtyPropertiesName(component)"
    project_content_dir=ProjectContentDir()
    component:Light Component Object Reference="Get (a copy)"(Array=Get Components(), Dimension 1=@n29.Array Index)
   ) -> (then, ReturnValue, filtered_properties:Array of Strings) {inputs [project_content_dir, component], outputs [filtered_properties]}
   > n31 = Set properties_name(properties_name:Array of Strings=@n30.filtered_properties) -> (Output_Get:Array of Strings)
   > Set LiveText(LiveText:String) -> (Output_Get:String)
   > n32 = For Each Loop(Array:Array of Strings=@n31) -> (Array Element:String)
   Set ToolTipText(
    execute=@n32.Completed
    ToolTipText=Format Text!(Format="INVTEXT(\"Dirty Properties: {list}\")", list:String=Get LiveText() -> (LiveText:String)) -> (Result)
    self=@n29
   )
   n33 = Branch(execute=@n32.LoopBody, Condition="Equal (Integer)"(A=@n32.Array Index))
   > Set LiveText(LiveText=@n32)
   Set LiveText(execute=@n33.else, LiveText=Append!(A=Get LiveText(), B=", ", C:String=@n32) -> (ReturnValue))
   n34 = GetDisplayName(Object=@n28)
   n35 = Create Widget(
    execute=@n28.LoopBody
    Class=&EWBP_LightValidationToolText_C
    Text="To Text (String)"(InString=Split(SourceString=@n34, InStr=., SearchDir=FromEnd).RightS)
   )
   > Add Unique(TargetArray=Get TextComponents(), NewItem=@n35)
   > AddChild(self=Get ComponentsVerticalBox(), Content=@n35)
   GameDevelopersDir!() -> (ReturnValue:String)
  ]}
 OnResetProperty {nodes [OnResetProperty() {is=title K2Node_FunctionEntry}]}
]
variables [
 Actor {type=Actor, default=null}
 Components {type="array<LightComponent>", default []}
 Delay {type=number, default=0.1}
 DynaComp {type="array<LightComponent>", default []}
 EndRay {type=Vector}
 LiveText {type=string, default=""}
 OnResetProperty {default=()}
 properties_name {type="array<string>", default []}
 TextComponents {type="array<EWBP_LightValidationToolText_C>", default []}
]
widgets {
 info {rootWidgetClass=&Overlay, widgetCount=24}
 widgets [
  Overlay_0 {widgetClassPath=&Overlay, widget=Overlay_0}
  HorizontalBox_32 {
   widgetClassPath=&HorizontalBox
   parent=Overlay_0
   widget=HorizontalBox_32
   slot {class=OverlaySlot, properties {horizontalAlignment=HAlign_Fill, padding=<ltrb 0 0 0 3>}}
  }
  Border_44 {
   widgetClassPath=&Border
   parent=HorizontalBox_32
   widget=Border_44
   properties {clipping=ClipToBounds, verticalAlignment=VAlign_Center, brushColor=<rgba 0.0573 0.0573 0.0573 1>}
   slot {class=HorizontalBoxSlot, properties {padding=<ltrb 0 0 2 0>, size {sizeRule=Fill, value=1}}}
  }
  ActorNameText {
   widgetClassPath=&TextBlock
   bIsVariable=true
   parent=Border_44
   widget=ActorNameText
   properties {
    text="Actor: {name} "
    font {
     bForceMonospaced=false
     bMaterialIsStencil=false
     fontMaterial=null
     fontObject=&Roboto
     letterSpacing=0
     monospacedWidth=1
     size=11.25
     skewAmount=0
     typefaceFontName=Bold
     outlineSettings {
      bApplyOutlineToDropShadows=false
      bMiteredCorners=false
      bSeparateFillAlpha=false
      outlineMaterial=null
      outlineSize=0
      outlineColor=<rgba 0 0 0 1>
     }
    }
   }
   slot {class=BorderSlot, properties {verticalAlignment=VAlign_Center}}
  }
  Border_1 {
   widgetClassPath=&Border
   parent=HorizontalBox_32
   widget=Border_1
   properties {clipping=ClipToBounds, brushColor=<rgba 0.0573 0.0573 0.0573 1>}
   slot {class=HorizontalBoxSlot, properties {padding=<ltrb 0 0 2 0>}}
  }
  VerticalBox_0 {widgetClassPath=&VerticalBox, parent=Border_1, widget=VerticalBox_0, slot {class=BorderSlot}}
  TextBlock_7 {
   widgetClassPath=&TextBlock
   parent=VerticalBox_0
   widget=TextBlock_7
   properties {
    text=Components Name:
    font {
     bForceMonospaced=false
     bMaterialIsStencil=false
     fontMaterial=null
     fontObject=&Roboto
     letterSpacing=0
     monospacedWidth=1
     size=11.25
     skewAmount=0
     typefaceFontName=Bold
     outlineSettings {
      bApplyOutlineToDropShadows=false
      bMiteredCorners=false
      bSeparateFillAlpha=false
      outlineMaterial=null
      outlineSize=0
      outlineColor=<rgba 0 0 0 1>
     }
    }
   }
   slot {class=VerticalBoxSlot}
  }
  Border_3 {
   widgetClassPath=&Border
   parent=VerticalBox_0
   widget=Border_3
   properties {brushColor=<rgba 0.0573 0.0573 0.0573 1>}
   slot {class=VerticalBoxSlot, properties {padding=<ltrb 0 0 2 0>, size {sizeRule=Fill, value=1}}}
  }
  ComponentsVerticalBox {widgetClassPath=&VerticalBox, bIsVariable=true, parent=VerticalBox_0, widget=ComponentsVerticalBox, slot {class=VerticalBoxSlot}}
  TextBlock_1 {
   widgetClassPath=&TextBlock
   parent=ComponentsVerticalBox
   widget=TextBlock_1
   properties {
    text=Component Name
    colorAndOpacity {colorUseRule=UseColor_Specified, specifiedColor=<rgba 0.5938 0.5938 0.5938 1>}
    font {
     bForceMonospaced=false
     bMaterialIsStencil=false
     fontMaterial=null
     fontObject=&Roboto
     letterSpacing=0
     monospacedWidth=1
     size=9.75
     skewAmount=0
     typefaceFontName=Bold
     outlineSettings {
      bApplyOutlineToDropShadows=false
      bMiteredCorners=false
      bSeparateFillAlpha=false
      outlineMaterial=null
      outlineSize=0
      outlineColor=<rgba 0 0 0 1>
     }
    }
   }
   slot {class=VerticalBoxSlot}
  }
  TextBlock_4 {
   widgetClassPath=&TextBlock
   parent=ComponentsVerticalBox
   widget=TextBlock_4
   properties {
    text=Component Name
    colorAndOpacity {colorUseRule=UseColor_Specified, specifiedColor=<rgba 0.5938 0.5938 0.5938 1>}
    font {
     bForceMonospaced=false
     bMaterialIsStencil=false
     fontMaterial=null
     fontObject=&Roboto
     letterSpacing=0
     monospacedWidth=1
     size=9.75
     skewAmount=0
     typefaceFontName=Bold
     outlineSettings {
      bApplyOutlineToDropShadows=false
      bMiteredCorners=false
      bSeparateFillAlpha=false
      outlineMaterial=null
      outlineSize=0
      outlineColor=<rgba 0 0 0 1>
     }
    }
   }
   slot {class=VerticalBoxSlot}
  }
  TextBlock_5 {
   widgetClassPath=&TextBlock
   parent=ComponentsVerticalBox
   widget=TextBlock_5
   properties {
    text=Component Name
    colorAndOpacity {colorUseRule=UseColor_Specified, specifiedColor=<rgba 0.5938 0.5938 0.5938 1>}
    font {
     bForceMonospaced=false
     bMaterialIsStencil=false
     fontMaterial=null
     fontObject=&Roboto
     letterSpacing=0
     monospacedWidth=1
     size=9.75
     skewAmount=0
     typefaceFontName=Bold
     outlineSettings {
      bApplyOutlineToDropShadows=false
      bMiteredCorners=false
      bSeparateFillAlpha=false
      outlineMaterial=null
      outlineSize=0
      outlineColor=<rgba 0 0 0 1>
     }
    }
   }
   slot {class=VerticalBoxSlot}
  }
  TextBlock_6 {
   widgetClassPath=&TextBlock
   parent=ComponentsVerticalBox
   widget=TextBlock_6
   properties {
    text=Component Name
    colorAndOpacity {colorUseRule=UseColor_Specified, specifiedColor=<rgba 0.5938 0.5938 0.5938 1>}
    font {
     bForceMonospaced=false
     bMaterialIsStencil=false
     fontMaterial=null
     fontObject=&Roboto
     letterSpacing=0
     monospacedWidth=1
     size=9.75
     skewAmount=0
     typefaceFontName=Bold
     outlineSettings {
      bApplyOutlineToDropShadows=false
      bMiteredCorners=false
      bSeparateFillAlpha=false
      outlineMaterial=null
      outlineSize=0
      outlineColor=<rgba 0 0 0 1>
     }
    }
   }
   slot {class=VerticalBoxSlot}
  }
  Border {
   widgetClassPath=&Border
   parent=HorizontalBox_32
   widget=Border
   properties {clipping=ClipToBounds, brushColor=<rgba 0.0573 0.0573 0.0573 1>}
   slot {class=HorizontalBoxSlot, properties {padding=<ltrb 0 0 2 0>}}
  }
  VerticalBox {widgetClassPath=&VerticalBox, parent=Border, widget=VerticalBox, slot {class=BorderSlot}}
  TextBlock {
   widgetClassPath=&TextBlock
   parent=VerticalBox
   widget=TextBlock
   properties {
    text=Last Change By:
    font {
     bForceMonospaced=false
     bMaterialIsStencil=false
     fontMaterial=null
     fontObject=&Roboto
     letterSpacing=0
     monospacedWidth=1
     size=11.25
     skewAmount=0
     typefaceFontName=Bold
     outlineSettings {
      bApplyOutlineToDropShadows=false
      bMiteredCorners=false
      bSeparateFillAlpha=false
      outlineMaterial=null
      outlineSize=0
      outlineColor=<rgba 0 0 0 1>
     }
    }
   }
   slot {class=VerticalBoxSlot}
  }
  UserVertical {
   widgetClassPath=&VerticalBox
   parent=VerticalBox
   widget=UserVertical
   slot {class=VerticalBoxSlot, properties {size {sizeRule=Fill, value=1}}}
  }
  UserNameText {
   widgetClassPath=&TextBlock
   bIsVariable=true
   parent=UserVertical
   widget=UserNameText
   properties {
    justification=Center
    text=Querying...
    colorAndOpacity {colorUseRule=UseColor_Specified, specifiedColor=<rgba 0.5938 0.5938 0.5938 1>}
    font {
     bForceMonospaced=false
     bMaterialIsStencil=false
     fontMaterial=null
     fontObject=&Roboto
     letterSpacing=0
     monospacedWidth=1
     size=9.75
     skewAmount=0
     typefaceFontName=Bold
     outlineSettings {
      bApplyOutlineToDropShadows=false
      bMiteredCorners=false
      bSeparateFillAlpha=false
      outlineMaterial=null
      outlineSize=0
      outlineColor=<rgba 0 0 0 1>
     }
    }
   }
   slot {class=VerticalBoxSlot, properties {verticalAlignment=VAlign_Center, size {sizeRule=Fill, value=1}}}
  }
  Border_2 {
   widgetClassPath=&Border
   parent=HorizontalBox_32
   widget=Border_2
   properties {horizontalAlignment=HAlign_Center, verticalAlignment=VAlign_Center, brushColor=<rgba 0.0573 0.0573 0.0573 1>}
   slot {class=HorizontalBoxSlot, properties {padding=<ltrb 0 0 2 0>}}
  }
  HorizontalBox_34 {
   widgetClassPath=&HorizontalBox
   parent=Border_2
   widget=HorizontalBox_34
   slot {class=BorderSlot, properties {horizontalAlignment=HAlign_Center, verticalAlignment=VAlign_Center}}
  }
  EditorUtilityButton_270 {
   widgetClassPath=&EditorUtilityButton
   parent=HorizontalBox_34
   widget=EditorUtilityButton_270
   slot {class=HorizontalBoxSlot}
  }
  Image_175 {
   widgetClassPath=&Image
   parent=EditorUtilityButton_270
   widget=Image_175
   properties {brush {
     bIsDynamicallyLoaded=false
     drawAs=Image
     imageType=FullColor
     mirroring=NoMirror
     resourceName=None
     resourceObject=/Game/Athena/UI/Challenges/Art/ChallengeTypeIcons/UI_ChallengeIcon_Find
     tiling=NoTile
     imageSize=<32 32>
     margin=<ltrb 0 0 0 0>
     outlineSettings {
      bUseBrushTransparency=false
      roundingType=HalfHeightRadius
      width=0
      color {colorUseRule=UseColor_Specified, specifiedColor=<rgba 0 0 0 0>}
      cornerRadii=<0 0 0 0>
     }
     tintColor {colorUseRule=UseColor_Specified, specifiedColor=<rgba 1 1 1 1>}
     uVRegion {bIsValid=false, max=<0 0>, min=<0 0>}
    }}
   slot {class=ButtonSlot}
  }
  EditorUtilityButton {
   widgetClassPath=&EditorUtilityButton
   parent=HorizontalBox_34
   widget=EditorUtilityButton
   properties {visibility=Collapsed}
   slot {class=HorizontalBoxSlot}
  }
  TextBlock_3 {
   widgetClassPath=&TextBlock
   parent=EditorUtilityButton
   widget=TextBlock_3
   properties {
    text=Restore
    font {
     bForceMonospaced=false
     bMaterialIsStencil=false
     fontMaterial=null
     fontObject=&Roboto
     letterSpacing=0
     monospacedWidth=1
     size=11.25
     skewAmount=0
     typefaceFontName=Bold
     outlineSettings {
      bApplyOutlineToDropShadows=false
      bMiteredCorners=false
      bSeparateFillAlpha=false
      outlineMaterial=null
      outlineSize=0
      outlineColor=<rgba 0 0 0 1>
     }
    }
   }
   slot {class=ButtonSlot, properties {horizontalAlignment=HAlign_Fill, padding=<ltrb 0 0 0 0>}}
  }
 ]
}
