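# Editor utility widget that searches for light components with inconsistent properties and lists them by owning actor.
# A search runs Python LightValidatorUtils to get dirty components, groups them into an actor-to-components map, then sets a short timer.
# StartCreateList creates an EWBP_LightActorElement per actor with an index-staggered delay and binds its reset event.
# A collapsed RestoreAll button resets the Python-defined light properties on every dirty component, then refreshes.
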EditorUtilityWidgetBlueprint /Game/NeverCook/ArtTools/Tech_Art/LightValidationTool/LightValidationTool
graphs [EventGraph {nodes [
   n1 = "On Clicked (EditorUtilityButton_32)"() {is=title K2Node_ComponentBoundEvent}
   Clear!(
    execute=[@n1.then, @n13.then]
    TargetMap:Map of Actor Object References to Struct Light Component Array Structures=Get ActorsComp() -> (ActorsComp:Map of Actor Object References to Struct Light Component Array Structures)
   ) -> (then) {is=title K2Node_CallFunction}
   > ClearChildren(self=Get ListOfActors() -> (ListOfActors:Vertical Box Object Reference))
   > n2 = Execute Python Script!(
    execute
    PythonScript="import sys\r\npath = project_content_dir+\"/Python/TechArt/Tools\"\r\nsys.path.append(project_content_dir+\"&Tools/\")\r\nimport LightValidatorUtils\r\nimport importlib\r\nimportlib.reload(LightValidatorUtils)\r\ncomponents_filtered = LightValidatorUtils.GetAllDirtyActorsComponents()"
    project_content_dir:String=ProjectContentDir!() -> (ReturnValue:String)
   ) -> (then, ReturnValue, components_filtered:Array of Light Component Object References) {inputs [project_content_dir], outputs [components_filtered]}
   > n3 = Set Components(Components:Array of Light Component Object References=@n2.components_filtered) -> (Output_Get:Array of Light Component Object References)
   > n4 = For Each Loop(Array:Array of Light Component Object References=@n3) -> (Array Element:Light Component Object Reference)
   Set Timer by Event(execute=@n4.Completed, Delegate=Create Event(), Time=0.1, bLooping=false, bMaxOncePerFrame=true)
   n5 = GetOwner(self=@n4)
   n6 = Find(
    TargetMap:Map of Actor Object References to Struct Light Component Array Structures=Get ActorsComp()
    Key:"Actor Object Reference (by ref)"=@n5
   ) -> (Value:Struct Light Component Array Structure)
   n7 = Branch(execute=@n4.LoopBody, Condition=@n6.ReturnValue)
   > n8 = Set TempLightComponents(TempLightComponents:Array of Light Component Object References=Break Struct Light Component Array!(Struct_LightComponentArray:"Struct Light Component Array Structure (by ref)"=@n6) -> (LightComponents_4_#1:Array of Light Component Object References) {is=title K2Node_BreakStruct}) -> (Output_Get:Array of Light Component Object References)
   > Remove!(
    execute
    TargetMap:Map of Actor Object References to Struct Light Component Array Structures=Get ActorsComp()
    Key:"Actor Object Reference (by ref)"=@n5
   ) -> (then, ReturnValue:bool)
   > Add Unique(TargetArray:Array of Light Component Object References=@n8, NewItem:"Light Component Object Reference (by ref)"=@n4)
   > Add!(
    execute
    TargetMap:Map of Actor Object References to Struct Light Component Array Structures=Get ActorsComp()
    Key:"Actor Object Reference (by ref)"=@n5
    Value:"Struct Light Component Array Structure (by ref)"=Make Struct Light Component Array!(LightComponents_4_#1:Array of Light Component Object References=Get TempLightComponents() -> (TempLightComponents:Array of Light Component Object References)) -> (Struct_LightComponentArray:Struct Light Component Array Structure) {is=title K2Node_MakeStruct}
   ) -> (then) {is=title K2Node_CallFunction}
   Add(
    execute=@n7.else
    TargetMap=Get ActorsComp()
    Key=@n5
    Value=Make Struct Light Component Array(LightComponents_4_#1=Make Array("[0]":Light Component Object Reference=@n4) -> (Array:Array of Light Component Object References))
   )
   "On Clicked (RestoreAllButton)"() {is=title K2Node_ComponentBoundEvent}
   > n9 = Execute Python Script(
    PythonScript="import sys\r\npath = project_content_dir+\"&Tools\"\r\nsys.path.append(project_content_dir+\"&Tools/\")\r\nimport LightValidatorUtils\r\nimport importlib\r\nimportlib.reload(LightValidatorUtils)\r\ncomponents_filtered = LightValidatorUtils.GetAllDirtyActorsComponents()"
    project_content_dir=ProjectContentDir()
   ) -> (components_filtered:Array of Point Light Component Object References) {inputs [project_content_dir], outputs [components_filtered]}
   > n10 = For Each Loop(Array:Array of Point Light Component Object References=@n9.components_filtered) -> (Array Element:Point Light Component Object Reference)
   UpdateActorList(execute=@n10.Completed)
   n11 = Execute Python Script!(
    execute=@n10.LoopBody
    PythonScript="import sys\r\npath = project_content_dir+\"&Tools\"\r\nsys.path.append(project_content_dir+\"&Tools/\")\r\nimport LightValidatorUtils\r\nproperties_name = LightValidatorUtils.light_properties"
    project_content_dir=ProjectContentDir()
   ) -> (then, ReturnValue, propertiesName:Array of Strings) {inputs [project_content_dir], outputs [propertiesName]}
   > n12 = For Each Loop(Array:Array of Strings=@n11.propertiesName) -> (Array Element:String)
   > ResetEditorProperty!(
    execute
    Object:Object Reference=@n10
    PropertyName:Name=String To Name(InString=@n12)
    ChangeNotifyMode:EPropertyAccessChangeNotifyMode Enum
   ) -> (then, ReturnValue:bool)
   n13 = UpdateActorList() {is=title K2Node_CustomEvent}
   StartCreateList() {is=title K2Node_CustomEvent}
   > n14 = Keys!(execute, TargetMap:Map of Actor Object References to Struct Light Component Array Structures=Get ActorsComp()) -> (then, Keys:Array of Actor Object References)
   > n15 = For Each Loop(Array:Array of Actor Object References=@n14) -> (Array Element:Actor Object Reference)
   n16 = int + int(A=@n15.Array Index)
   > n17 = Create Widget!(
    execute:Exec
    WorldContextObject:Object Reference
    Class:User Widget Class Reference=&/EWBP_LightActorElement.EWBP_LightActorElement_C
    OwningPlayer:Player Controller Object Reference
    Components:Array of Light Component Object References=Find!(TargetMap=Get ActorsComp(), Key=@n15) -> (Value_LightComponents_4_#1:Array of Light Component Object References, ReturnValue)
    Actor:Actor Object Reference=@n15
    Delay:double=SafeDivide(A="To Float (Integer)"(InInt=@n16), B=2)
   ) -> (then:Exec, ReturnValue:EWBP Light Actor Element Object Reference)
   > Bind Event to On Reset Property(self:EWBP Light Actor Element Object Reference=@n17, Delegate=Create Event()) {is=title K2Node_AddDelegate}
   > AddChild(self=Get ListOfActors(), Content=@n17)
   GameDevelopersDir!() -> (ReturnValue:String)
  ]}]
variables [
 ActorsComp {default {}}
 Components {type="array<LightComponent>", default []}
 DynaComp {type="array<LightComponent>", default []}
 TempLightComponents {type="array<LightComponent>", default []}
]
widgets {
 info {rootWidgetClass=&Overlay, widgetCount=9}
 widgets [
  Overlay_0 {widgetClassPath=&Overlay, widget=Overlay_0}
  VerticalBox_125 {
   widgetClassPath=&VerticalBox
   parent=Overlay_0
   widget=VerticalBox_125
   slot {class=OverlaySlot, properties {horizontalAlignment=HAlign_Fill, verticalAlignment=VAlign_Fill}}
  }
  HorizontalBox_0 {
   widgetClassPath=&HorizontalBox
   parent=VerticalBox_125
   widget=HorizontalBox_0
   slot {class=VerticalBoxSlot, properties {horizontalAlignment=HAlign_Left}}
  }
  EditorUtilityButton_32 {widgetClassPath=&EditorUtilityButton, parent=HorizontalBox_0, widget=EditorUtilityButton_32, slot {class=HorizontalBoxSlot}}
  TextBlock_60 {
   widgetClassPath=&TextBlock
   parent=EditorUtilityButton_32
   widget=TextBlock_60
   properties {
    text=Search for inconsistency
    font {
     bForceMonospaced=false
     bMaterialIsStencil=false
     fontMaterial=null
     fontObject=&Roboto
     letterSpacing=0
     monospacedWidth=1
     size=12
     skewAmount=0
     typefaceFontName=Bold
     outlineSettings {
      bApplyOutlineToDropShadows=false
      bMiteredCorners=false
      bSeparateFillAlpha=false
      outlineMaterial=null
      outlineSize=0
      outlineColor=<rgba 0 0 0 1>
     }
    }
   }
   slot {class=ButtonSlot}
  }
  RestoreAllButton {
   widgetClassPath=&EditorUtilityButton
   parent=HorizontalBox_0
   widget=RestoreAllButton
   properties {visibility=Collapsed}
   slot {class=HorizontalBoxSlot, properties {horizontalAlignment=HAlign_Left}}
  }
  RestoreAllText {
   widgetClassPath=&TextBlock
   parent=RestoreAllButton
   widget=RestoreAllText
   properties {
    text=RestoreAll
    font {
     bForceMonospaced=false
     bMaterialIsStencil=false
     fontMaterial=null
     fontObject=&Roboto
     letterSpacing=0
     monospacedWidth=1
     size=12
     skewAmount=0
     typefaceFontName=Bold
     outlineSettings {
      bApplyOutlineToDropShadows=false
      bMiteredCorners=false
      bSeparateFillAlpha=false
      outlineMaterial=null
      outlineSize=0
      outlineColor=<rgba 0 0 0 1>
     }
    }
   }
   slot {class=ButtonSlot}
  }
  EditorUtilityScrollBox_99 {
   widgetClassPath=&EditorUtilityScrollBox
   parent=VerticalBox_125
   widget=EditorUtilityScrollBox_99
   slot {class=VerticalBoxSlot, properties {verticalAlignment=VAlign_Top, size {sizeRule=Fill, value=1}}}
  }
  ListOfActors {widgetClassPath=&VerticalBox, bIsVariable=true, parent=EditorUtilityScrollBox_99, widget=ListOfActors, slot {class=ScrollBoxSlot}}
 ]
}
